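# Subsurface monument material for the S10 Keving memorial quest interactable, built on PowerPlant monument textures with a purple glow.
# Emissive uses inverted diffuse alpha, masked by inverse Fresnel and an offset sphere, flickered by time-sampled noise and weighted by time of day.
# Base color blends a light-reflection tint; subsurface glow follows a height gradient.
# Deferred adds normals and shared hit, tree, morph WPO; Forward/Mobile omit these and derive roughness from diffuse.
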Material /Game/Athena/Items/QuestInteractables/S10_KevingMemorial/M_Monument_12
graph {
 properties {
  bUseMaterialAttributes=true
  bUsedWithNanite=true
  pixelDepthOffsetMode=PDOM_Legacy
  shadingModel=MSM_Subsurface
  usedShadingModels=Subsurface
 }
 nodes [
  n1 = VectorParameter() {parameterName=Light Color, defaultValue=<rgba 0 0.16 0.7552 0>}
  n2 = TextureSample() {texture=/Game/Environments/Apollo/Sets/PowerPlant/Props/Textures/T_PowerPlant_Monument_01_D}
  n3 = DotProduct(A=AtmosphericLightVector(), B=ReflectionVectorWS())
  n4 = LinearInterpolate(A=Multiply(A=LinearInterpolate(A=@n1.RGB, B=@n2.RGB), B=Multiply(A=@n3, B=@n2.A)), B=@n2.RGB, Alpha=@n2.RGB)
  n5 = TextureSample() {samplerType=SAMPLERTYPE_Masks, texture=/Game/Environments/Apollo/Sets/PowerPlant/Props/Textures/T_PowerPlant_Monument_01_S}
  n6 = VectorParameter() {parameterName=GlowColor, defaultValue=<rgba 0.4216 0.0154 0.4219 0>}
  n7 = Multiply(A=OneMinus(None=@n2.A), B=@n6.RGB)
  n8 = VertexNormalWS()
  n9 = WorldPosition()
  n10 = ActorPositionWS()
  n11 = VectorParameter() {parameterName=GlowOffset, defaultValue=<rgba 0 0 249.9492 0>}
  n12 = Multiply(A=OneMinus(None=Fresnel(Normal=@n8) {exponent=3}), B=SphereMask(A=@n9.XYZ, B=Add(A=@n10, B=@n11.RGB), Hardness=Constant()))
  n13 = Multiply(A=Time(), B=Constant() {r=0.125})
  n14 = CollectionParameter() {
   collection=&FortniteMaterialParameters
   parameterName=FortniteActiveTimeOfDay
  }
  n15 = Multiply(
   A=Multiply(
    A=Multiply(A=@n7, B=@n12)
    B=TextureSample(UVs=@n13) {samplerType=SAMPLERTYPE_LinearGrayscale, texture=/Game/Global/Textures/Default/Noise/T_SwirlNoise_N_height}.R
   )
   B=DotProduct(A=@n14, B=Constant4Vector() {constant=<rgba 0.75 0.75 0.75 1>}.RGBA)
  )
  n16 = Constant() {r=1}
  n17 = Add(
   A=MaterialFunctionCall!(
    WSDeformationCenter=VectorParameter() {bUseCustomPrimitiveData=true, group=z_code, parameterName=WSDeformationCenter, primitiveDataIndex=20}.RGB
    Code_DestructionAmount=ScalarParameter() {group=z_code, parameterName=Code_DestructionAmount}
    Code_UseUvForDamage
    Code_UseHitEffects=StaticBoolParameter() {defaultValue=true, group=z_code, parameterName=Code_UseHitEffects}
    Code_Emissive
    WSDamageRadius=ScalarParameter() {bUseCustomPrimitiveData=true, defaultValue=1, group=z_code, parameterName=WSDamageRadius, primitiveDataIndex=26}
    CrackDamageShader
    Diffuse
    ScalarModDamageCracks
    ScalarOffsetforDamageCracks
    ExposeLerpFunctionForInvertingTheHitDistanceCalc
    InvertDamageFalloff
    CrackWorldPosition
    WorldPosition
    DistanceCullFadeAnimationOffset=ScalarParameter() {bUseCustomPrimitiveData=true, group=z_code, parameterName=DistanceCullFadeAnimationOffset, primitiveDataIndex=12}
   ) {materialFunction=/Game/Packages/jl_tests/Fortnite_HitEffects}.Hit And Shrink WPO
   B=MaterialFunctionCall!(Use Tree Vertex Shader=StaticBoolParameter() {group=Special Cases, parameterName=Use Tree Vertex Shader}) {materialFunction=/Game/Packages/Fortress_SharedMaterials/Base_Material_Functions/FortniteTreeAnimation_VS}.Result
  )
  n18 = MaterialFunctionCall() {materialFunction=/Engine/Functions/Engine_MaterialFunctions02/WorldPositionOffset/StaticMeshMorphTargets}
  n19 = ScalarParameter() {group=MorphTargets, parameterName=MorphTargetStrength}
  n20 = StaticSwitchParameter(True=Add(A=@n17, B=Multiply(A=@n18.Morph Target 1 WPO, B=@n19)) {desc=7 instructions}, False=@n17) {group=MorphTargets, parameterName=UseMorphTargets}
  n21 = VectorParameter() {group=WorldPositionOffset, parameterName=WorldPositionOffset_ModulatedByVertexColor, defaultValue=<rgba 0 0 0 1>}
  n22 = VertexColor()
  n23 = Multiply(
   A=@n21.RGB
   B=StaticComponentMaskParameter(None=AppendVector(A=@n22, B=@n22)) {defaultA=true, group=WorldPositionOffset, parameterName=WorldPositionOffsetVertexColorModulationChannel}
  )
  n24 = Add(A=ComponentMask(None=Subtract(A=@n10, B=@n9.XYZ)) {b=true}, B=Constant() {r=144.6559})
  n25 = Constant() {r=184.0657}
  n26 = LinearInterpolate(A=Multiply(A=Multiply(A=@n2.A, B=@n6.RGB), B=ScalarParameter() {defaultValue=0.1619}), Alpha=Saturate(None=Divide(A=@n24, B=@n25))) {constB=0}
  n27 = SetMaterialAttributes!(
   MaterialAttributes
   Base Color=@n4
   Specular=Multiply(A=@n5.R, B=OneMinus(None=@n15))
   Roughness=@n5.B
   Emissive Color=@n15
   Opacity=@n16
   Normal=TextureSample() {samplerType=SAMPLERTYPE_Normal, texture=/Game/Environments/Apollo/Sets/PowerPlant/Props/Textures/T_PowerPlant_Monument_01_N}.RGB
   World Position Offset=StaticSwitchParameter(True=Add(A=@n20, B=@n23), False=@n20) {group=WorldPositionOffset, parameterName=UseWorldPositionOffsetModulatedByVertexColors}
   Subsurface Color=@n26
  ) {attributeSetTypes [
    #1
    #2
    #3
    #4
    #5
    #6
    #7
    #8
   ]}
  n28 = Multiply(A=Constant() {r=20}, B=@n2.R) {constB=3}
  n29 = SetMaterialAttributes(
   Base Color=@n4
   Roughness=LinearInterpolate(A=Saturate(None=@n28), Alpha=Constant() {r=0.6})
   Emissive Color=@n15
   Opacity=@n16
   Subsurface Color=@n26
  ) {attributeSetTypes [
    #1
    #2
    #3
    #4
    #5
    #6
    #7
    #8
   ]}
  ShadingPathSwitch(Default=@n27, Deferred=@n27, Forward=@n29, Mobile=@n29)
 ]
}
